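# Physics asset for the Lemon Lime backpack cosmetic mesh, defining simulated motion for its dangling attachments.
# Box bodies cover the pack, handle, side belt and zipper bones, with the small pack box acting as the hub.
# Each of the three dyn_ bones is constrained directly to the pack with tight twist and asymmetric swing limits, so each pivots mostly within its own restricted cone.
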PhysicsAsset /BRCosmetics/Accessories/FORT_Backpacks/F_MED_Lemon_Lime_Backpack/Meshes/F_MED_Lemon_Lime_Pack_Physics
bodies {
 dyn_handel [{shapeType=Box, extentX=0.5, extentY=3.3478, extentZ=8.5246, rotation=<rot -14.818 0 0>}]
 dyn_side_belt [{shapeType=Box, extentX=12.9153, extentY=2.9669, extentZ=2.7932, center=<5.651 0 -0.1322>}]
 dyn_zipper [{shapeType=Box, extentX=5.01, extentY=2.51, extentZ=1.01, center=<1.8948 0 0>}]
 pack [{shapeType=Box, extentX=1.5, extentY=1.5, extentZ=1.5}]
}
constraints [
 {bone1Name=dyn_side_belt, bone2Name=pack, swing1LimitDegrees=10, swing2LimitDegrees=60, twistLimitDegrees=3}
 {bone1Name=dyn_zipper, bone2Name=pack, swing1LimitDegrees=15, swing2LimitDegrees=35, twistLimitDegrees=3}
 {bone1Name=dyn_handel, bone2Name=pack, swing1LimitDegrees=30, swing2LimitDegrees=20, twistLimitDegrees=5}
]
