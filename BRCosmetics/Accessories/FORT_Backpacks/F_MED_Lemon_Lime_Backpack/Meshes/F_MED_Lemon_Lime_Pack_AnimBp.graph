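# Anim Blueprint for the Lemon Lime backpack cosmetic that adds physics motion to the pack, with an extra belt kick on jumps.
# The input pose feeds a RigidBody node simulating the pack's physics asset in root-bone space with overridden gravity and scaled component motion.
# While bIsJumping is true, the event graph applies a delayed upward Belt_Ext_Force, then re-arms a Do Once that zeroes it when not jumping.
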AnimBlueprint /BRCosmetics/Accessories/FORT_Backpacks/F_MED_Lemon_Lime_Backpack/Meshes/F_MED_Lemon_Lime_Pack_AnimBp
target_skeleton=&/F_MED_Lemon_Lime_Pack_Skeleton
graphs [
 AnimGraph {nodes [
   AnimGraphNode_SubInput_0 = Input Pose()
   n1 = RigidBody(ComponentPose=Local To Component(LocalPose=@AnimGraphNode_SubInput_0), ExternalForce=Get Belt_Ext_Force() -> (Belt_Ext_Force:Vector)) {
    bClampLinearTranslationLimitToRefPose=true
    bOverrideWorldGravity=true
    overridePhysicsAsset=&/F_MED_Lemon_Lime_Pack_Physics
    simulationSpace=BaseBoneSpace
    alphaBoolBlend {blendOption=Linear}
    baseBoneRef {boneName=root}
    componentAppliedLinearAccClamp=<600 600 600>
    componentLinearAccScale=<0.25 0.25 0.25>
    componentLinearVelScale=<0.25 0.25 0.25>
    overrideWorldGravity {z=-980}
   }
   Output Pose(Result=Component To Local(ComponentPose=@n1)) {layerGroup=None}
  ]}
 EventGraph {nodes [
   Event BlueprintUpdateAnimation()
   > n1 = Branch(Condition=Get bIsJumping())
   > Delay(Duration=0.25)
   > Set Belt_Ext_Force(Belt_Ext_Force:Vector="0,0,1250") -> (Output_Get:Vector)
   > n2 = Delay(Duration=0.35)
   Do Once(execute=@n1.else, Reset=@n2.then)
   > Set Belt_Ext_Force(Belt_Ext_Force="0, 0, 0")
   TryGetPawnOwner()
  ]}
]
variables [Belt_Ext_Force {type=Vector}]
